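; Feather resource pickup actor that varies its look and sound with the amount it holds.
; On attach, clients read the count from the parent FortPickup's item entry and pick a small, medium or large mesh and sounds via ScalableFloat thresholds.
; Pickup spawns the selected sound at the actor, and entering the backpack attaches that audio component to the pawn's root.
; The glow, drop-sound and item-entry-array hooks are empty here.
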(event EventOnAttached
  ;[0.4437 1 0.3343] SPAWN: Get resource count.
  (|CallParentFunction)
  (Networking|CanExecuteCosmeticEvents
    (:True
      (bind _asfort_pickup (Utilities|Casting|CastToFortPickup (Actor|GetAttachParentActor)))
      (Variables|PersistentPickupProperties|SetPickupCount (.Count (Class|FortItemFunctionLibrary|BreakFortItemEntry (Pickup|GetItemEntry _asfort_pickup 0))))
      ;[0.4437 1 0.3343] Set cosmetics and play drop sound.
      (CallFunction|SetCosmeticsBasedOnCount)
      (CallFunction|PlayDropSound))))

(event EventOnAboutToEnterBackpack
  (params (PickupTarget :type Fort%20Pawn%20Object%20Reference))
  ;[0.3812 0.4441 1] PICKUP COMPLETE: attach sound component to pawn, send up MCP loot.
  (CallFunction|SpawnPickupSoundComponent 0))

(event EventOnPickedUp
  (params (PickupTarget :type Fort%20Pawn%20Object%20Reference))
  ;[1 0.7077 0.2831] SOUND ON PICKUP AND DROP
  (|CallParentFunction 0)
  (bind _returnvalue (Audio|SpawnSoundatLocation (Variables|PersistentPickupProperties|GetSelected%20Pickup%20Sound) (Transformation|GetActorLocation)))
  (Variables|Default|SetPickup%20Audio%20Component _returnvalue))

(event Custom|Play%20Drop%20Sound)

(fn MakeItemEntryArray)

(fn SetCosmeticsBasedonCount
  (bind _large_threshold (Variables|PersistentPickupProperties|GetLarge%20Threshold))
  (bind _returnvalue (ScalableFloat|GetValueatLevel _large_threshold))
  (bind _returnvalue_2 (Math|Float|Truncate _returnvalue))
  ;[1 1 1] Large
  (LocalVariables|SetNewLocalVar_0 _returnvalue_2)
  (if (>= (Variables|PersistentPickupProperties|GetPickupCount) _returnvalue_2)
    (Components|StaticMesh|SetStaticMesh (Variables|FortPickupsParent|GetStaticMeshComponent) (Variables|PersistentPickupProperties|GetMesh%20Lrg))
    (Transformation|SetWorldScale3D (Variables|FortPickupsParent|GetStaticMeshComponent) (Math|Conversions|ToVector(Float) (Variables|PersistentPickupProperties|GetMesh%20Scale)))
    (CallFunction|SetupPickupGlow 2)
    (Variables|PersistentPickupProperties|SetSelected%20Pickup%20Sound (Variables|Sounds|GetPickup%20Sound%20Large))
    (Audio|Components|Audio|SetSound (Variables|Default|GetDrop%20Audio%20Component) (Variables|Sounds|GetDrop%20Sound%20Large))
    (else
      ;[1 1 1] Med
      (if (>= (Variables|PersistentPickupProperties|GetPickupCount) (Math|Float|Truncate (ScalableFloat|GetValueatLevel (Variables|PersistentPickupProperties|GetMedium%20Threshold))))
        (Components|StaticMesh|SetStaticMesh (Variables|FortPickupsParent|GetStaticMeshComponent) (Variables|PersistentPickupProperties|GetMesh%20Med))
        (Transformation|SetWorldScale3D (Variables|FortPickupsParent|GetStaticMeshComponent) (Math|Conversions|ToVector(Float) (Variables|PersistentPickupProperties|GetMesh%20Scale)))
        (CallFunction|SetupPickupGlow 1)
        (Variables|PersistentPickupProperties|SetSelected%20Pickup%20Sound (Variables|Sounds|GetPickup%20Sound%20Medium))
        (Audio|Components|Audio|SetSound (Variables|Default|GetDrop%20Audio%20Component) (Variables|Sounds|GetDrop%20Sound%20Medium))
        (else
          ;[1 1 1] Small
          (Components|StaticMesh|SetStaticMesh (Variables|FortPickupsParent|GetStaticMeshComponent) (Variables|PersistentPickupProperties|GetMesh%20Small))
          (Transformation|SetWorldScale3D (Variables|FortPickupsParent|GetStaticMeshComponent) (Math|Conversions|ToVector(Float) (Variables|PersistentPickupProperties|GetMesh%20Scale)))
          (CallFunction|SetupPickupGlow 0)
          (Variables|PersistentPickupProperties|SetSelected%20Pickup%20Sound (Variables|Sounds|GetPickup%20Sound%20Small))
          (Audio|Components|Audio|SetSound (Variables|Default|GetDrop%20Audio%20Component) (Variables|Sounds|GetDrop%20Sound%20Small)))))))

(fn SetupPickupGlow
  (params (Level :type Integer)))

(fn SpawnPickupSoundComponent
  (params (Pawn%20to%20Attach%20To :type Fort%20Pawn%20Object%20Reference))
  (bind _pickup_audio_component (Variables|Default|GetPickup%20Audio%20Component))
  (Utilities|IsValid Pawn%20to%20Attach%20To
    (:Is%20Valid
      (Utilities|IsValid _pickup_audio_component
        (:Is%20Valid
          (Transformation|AttachComponentToComponent _pickup_audio_component (Variables|/Script/Engine.Actor|GetRootComponent Pawn%20to%20Attach%20To) "root" "SnapToTarget" "SnapToTarget" "SnapToTarget" true))))))

(fn ConstructionScript
  (|CallParentFunction)
  (Variables|Unique_Material_Check|SetHasUniqueMaterialIds true))
