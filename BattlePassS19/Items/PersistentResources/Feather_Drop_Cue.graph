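# Sound cue for dropping the Feather persistent resource item from Battle Pass Season 19.
# A single wave player plays Feather_Drop_01 at reduced volume through the loot collect sound class.
# It uses weapon drop attenuation with overridden falloff and stereo spread, and concurrency overrides allow only one instance, preventing new plays while one is active.
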SoundCue /BattlePassS19/Items/PersistentResources/Feather_Drop_Cue
root=WavePlayer_2
nodes [{id=WavePlayer_2, properties {soundWaveAssetPtr=&/Feather_Drop_01}}]
properties {
 attenuationSettings=&Fort_Weapon_Drop_Att
 maxDistance=3000
 priority=0.9
 soundClassObject=&Fort_Loot_Collect
 volumeMultiplier=0.279
 attenuationOverrides {falloffDistance=1000, stereoSpread=512, attenuationShapeExtents {x=0}}
 concurrencyOverrides {maxCount=1, resolutionRule=PreventNew}
 concurrencySet [null]
}
